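; Healing cactus actor that bursts when damaged, touched by a vehicle, or hit by sliding or tow-hooked players, healing its surroundings.
; Bursting fires a gameplay cue and applies Slurp effects to nearby actors, or vehicle passengers plus percentage vehicle healing.
; A timer regrows it after optional storm and overlap checks, retrying while blocked.
; RepNotifies drive collision, audio, FX and material custom primitive data; ScalableFloat hotfix values control tuning.
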(event EventBeginPlay
  (bind _fortsoundindicator (Variables|Default|GetFortSoundIndicator))
  ;[0 0.8244 1] Begin Play
  (|CallParentFunction)
  (CallFunction|FadeinAmbientLoop)
  (Components|Activation|Deactivate _fortsoundindicator)
  (SoundIndicator|StopSoundIndicatorLooping _fortsoundindicator)
  (if (not (ScalableFloat|GetValueasBool (Variables|Hotfix|GetHF_ShouldObstructCamera)))
    (Utilities|Array|AddUnique (Variables|Actor|GetTags) (Variables|Default|GetCameraCollisionTag))))

(event EventOnDamageServer
  (params (Damage :type Float%20%28single-precision%29) (DamageTags :type Gameplay%20Tag%20Container%20Structure%20%28by%20ref%29) (Momentum :type Vector :default "0, 0, 0") (HitInfo :type Hit%20Result%20Structure%20%28by%20ref%29) (InstigatedBy :type Controller%20Object%20Reference) (DamageCauser :type Actor%20Object%20Reference) (EffectContext :type Gameplay%20Effect%20Context%20Handle%20Structure))
  ;[0.4107 0.6115 0.7899] On damaged authority
  (if (not (Variables|Default|GetbWaitingToRegrow))
    (CallFunction|ResetHealth)
    (bind _asfort_weapon (Utilities|Casting|CastToFortWeapon DamageCauser)
      (:then
        (Variables|Default|SetInstigatingActor (Actor|GetOwner _asfort_weapon)))
      (:CastFailed
        (Default|SetInstigatingActorwithCheck DamageCauser)))
    (CallFunction|StartBurst false)))

(event Custom|StartBurst
  (params (VehicleCaused :type Boolean))
  ;[0 0.8244 1] Start bursting sequence -- set variables and call healing if not triggered by vehicle (vehicle healing is done elsewhere)
  (Rendering|Material|SetCustomPrimitiveDataFloat (Variables|Default|GetLaunchableMesh) 2 (Math|Conversions|ToFloat(Boolean) VehicleCaused)) ; TechArt: set CPD for bulging/hide instantly based on VehicleCaused
  (Variables|Default|SetCollisionEnabled false)
  (Variables|Default|SetbWaitingToRegrow true)
  (bind _returnvalue (Utilities|Time|SetTimerbyEvent (EventDispatchers|CreateEvent :!event Custom|TryRegrowth) (ScalableFloat|GetValueatLevel (Variables|Hotfix|GetTimeToRegrowth))))
  (Variables|Default|SetRegrowthCheckTimerHandle _returnvalue)
  (if (not VehicleCaused)
    (CallFunction|DoHealing)))

(event Custom|DoHealing
  ;[1 1 1] Apply Heal to Actors + Healing GCNs
  (bind _outactors (CallFunction|DoTargetSelection))
  (bind _output_get (Variables|Default|SetOut%20Actors _outactors))
  (if (Utilities|Array|IsNotEmpty _output_get)
    ;;[1 1 1] Spawn explode VFX Visuals reguardless of whether a Pawn is inside the trigger box
    (Utilities|FlowControl|Delay 0.2)
    (CallFunction|BurstCue)
    (Utilities|FlowControl|Delay 0.3)
    (for _array_element (Variables|Default|GetOut%20Actors)
      (CallFunction|ApplySlurp _array_element))
    (else
      (Utilities|FlowControl|Delay 0.2)
      (CallFunction|BurstCue))))

(event Custom|TryRegrowth
  ;[0.03429 0.3069 0.7899] Regrowth functionality -- reset everything as if the cactus had just spawned for the first time
  (bind _ballowregrow (CallFunction|CanRegrow))
  (if _ballowregrow
    (Variables|Default|SetbWaitingToRegrow false)
    (Variables|Default|SetCollisionEnabled true)
    (Variables|Default|SetVehicle%20Caused false)
    (Variables|Default|SetSlidingOrTowHooked false)
    (Rendering|Material|SetCustomPrimitiveDataFloat (Variables|Default|GetLaunchableMesh) 2 0.0) ; TechArt: Reset CPD
  ))

(event OnComponentBeginOverlap(LaunchableMesh)
  (params (OverlappedComponent :type Primitive%20Component%20Object%20Reference) (OtherActor :type Actor%20Object%20Reference) (OtherComp :type Primitive%20Component%20Object%20Reference) (OtherBodyIndex :type Integer) (bFromSweep :type Boolean) (SweepResult :type Hit%20Result%20Structure%20%28by%20ref%29))
  ;[0.9394 0.2521 1] Vehicle interaction
  ;; To heal players in range outside the vehicle
  (CallFunction|DoHealing)
  (bind _asfort_vehicle_interface (Utilities|Casting|CastToFortVehicleInterface OtherActor))
  (Variables|Default|SetInstigatingActor OtherActor)
  (bind _output_get (Variables|Default|SetVehicle%20Caused true))
  (CallFunction|StartBurst _output_get)
  (CallFunction|BurstCue)
  (Utilities|FlowControl|Delay 0.2)
  (CallFunction|ApplySlurptoPassengers _asfort_vehicle_interface)
  (CallFunction|ApplyHealtoVehicle _asfort_vehicle_interface))

(event Custom|ApplySlurpToPassengers
  (params (FortVehicleInterface :type Fort%20Vehicle%20Interface%20Interface))
  ;[0.9394 0.2521 1] Vehicle interaction
  (for _array_element (Vehicle|GetAllPassengers FortVehicleInterface)
    (CallFunction|ApplySlurp _array_element)))

(event OnComponentHit(LaunchableMesh)
  (params (HitComponent :type Primitive%20Component%20Object%20Reference) (OtherActor :type Actor%20Object%20Reference) (OtherComp :type Primitive%20Component%20Object%20Reference) (NormalImpulse :type Vector :default "0, 0, 0") (Hit :type Hit%20Result%20Structure%20%28by%20ref%29))
  ;[0.9181 1 0] Slide kick and Tow Hook functionality
  (Utilities|DoOncePerFrame
    (:%20%20
      (if (ScalableFloat|GetValueasBool (Variables|Hotfix|GetHF_TowHookEnabled))
        (bind _returnvalue (GameplayTags|HasMatchingGameplayTag(Message) OtherActor "(TagName=\"Granted.Athena.Item.TowHookWeapon.PlayerTetheredAnyState\")"))
        (if _returnvalue
          _
          (elif (ScalableFloat|GetValueasBool (Variables|Hotfix|GetHF_SlideKickEnabled))
            (bind _returnvalue_2 (GameplayTags|HasMatchingGameplayTag(Message) OtherActor "(TagName=\"MovementMode.Sliding\")"))
            (if _returnvalue_2
              _)))
        (Variables|Default|SetInstigatingActor OtherActor)
        (CallFunction|StartBurst false)
        (else
          _)))))

(fn ApplySlurp
  (params (TargetPawn :type Object%20Reference))
  (Utilities|IsValid TargetPawn
    (:Is%20Valid
      (bind _asfort_pawn (Utilities|Casting|CastToFortPawn TargetPawn))
      (if (not (GameplayTags|HasAnyMatchingGameplayTags _asfort_pawn (Variables|Default|GetBlockingTags)))
        (Fortnite|ApplyGameplayEffectsToActor _asfort_pawn (Utilities|Array|MakeArray (Variables|Default|GetSlurp%20GE)) (Transformation|GetActorLocation))))))

(fn ApplyHealToVehicle
  (params (Fort%20Vehicle%20Interface :type Fort%20Vehicle%20Interface%20Interface))
  (bind _returnvalue (Vehicle|GetVehicleAbilitySystemComponent Fort%20Vehicle%20Interface))
  (bind _returnvalue_2 (Ability|GameplayEffect|AssignTagSetbyCallerMagnitude (GameplayEffects|MakeOutgoingSpec _returnvalue (Variables|Default|GetVehicle%20Heal%20GE) 1.0 (Class|AthenaFunctionLibrary|MakeSimpleFortEffectContext (Variables|Default|GetInstigatingActor))) "(TagName=\"SetByCaller.Healing\")" (* (Class|FortVehicleInterface|GetMaxHealth Fort%20Vehicle%20Interface) (ScalableFloat|GetValueatLevel (Variables|Hotfix|GetVehicleHealPercent)))))
  (GameplayEffects|ApplyGameplayEffectSpecToSelf _returnvalue _returnvalue_2))

(fn BurstCue
  (Ability|ExecuteGameplayCueWithParamsReplicated (Variables|Default|GetSlurpExplosionGCTag) (Ability|GameplayCue|MakeGameplayCueParameters :NormalizedMagnitude 0.0 :RawMagnitude 0.0 :Location (Transformation|GetActorLocation) :Normal "0,0,0" :Instigator (Variables|Default|GetInstigatingActor) :GameplayEffectLevel 1 :AbilityLevel 1 :bReplicateLocationWhenUsingMinimalRepProxy false)))

(fn CanRegrow
  (if (ScalableFloat|GetValueasBool (Variables|Hotfix|GetHF_RegrowEnabled))
    (if (not (ScalableFloat|GetValueasBool (Variables|Hotfix|GetHF_GrowInStorm?)))
      (bind _returnvalue (SafeZone|IsLocationinSafeZone (Transformation|GetActorLocation)))
      (if (not _returnvalue)
        (return false)))
    (bind _nooverlaps (CallFunction|PerformOverlapActorCheck))
    (return _nooverlaps)))

(fn DoTargetSelection
  (bind _outactors (Collision|SphereOverlapActors (Transformation|GetActorLocation) (* (ScalableFloat|GetValueatLevel (Variables|Hotfix|GetHF_HealRadius)) (Math|Vector|VectorLength (Transformation|GetActorScale3D))) (Variables|Default|GetObject%20Types) 0 (Utilities|Array|MakeArray self)))
  (return _outactors))

(fn FadeInAmbientLoop
  (Audio|Components|Audio|FadeIn (Variables|Default|GetAmbient%20Loop) 0.1 1.0 0.0 "Linear" :!class /Script/Engine.AudioComponent))

(fn FadeOutAmbientLoop
  (Audio|Components|Audio|FadeOut (Variables|Default|GetAmbient%20Loop) 0.5 0.0 "Linear" :!class /Script/Engine.AudioComponent)
  (SoundIndicator|StopSoundIndicatorLooping (Variables|Default|GetFortSoundIndicator)))

(fn OnRep_bWaitingToRegrow
  (bind _launchablemesh (Variables|Default|GetLaunchableMesh))
  (bind _launchablemesh_2 (Variables|Default|GetLaunchableMesh))
  (if (Variables|Default|GetbWaitingToRegrow)
    ;[1 1 1] Tech art for fruits burst
    (Rendering|Material|SetCustomPrimitiveDataFloat _launchablemesh_2 0 (Utilities|Time|GetGameTimeinSeconds))
    (Rendering|Material|SetCustomPrimitiveDataFloat _launchablemesh_2 1 -1.0)
    (CallFunction|FadeOutAmbientLoop)
    (else
      (Utilities|FlowControl|Sequence
        (:then_1
          ;[1 1 1] Tech art for fruits regrowth
          (Rendering|Material|SetCustomPrimitiveDataFloat _launchablemesh 0 (Utilities|Time|GetGameTimeinSeconds))
          (Rendering|Material|SetCustomPrimitiveDataFloat _launchablemesh 1 1.0)
          (CallFunction|FadeinAmbientLoop)
          (Audio|PlaySoundatLocation (Variables|Default|GetRegrowthSound) (Transformation|GetActorLocation))
          (Components|Activation|Activate (Variables|Default|GetRegrowthFX) true))))))

(fn OnRep_CollisionEnabled
  (if (Variables|Default|GetCollisionEnabled)
    (Collision|SetCollisionEnabled (Variables|Default|GetLaunchableMesh) "QueryAndPhysics")
    (else
      (Collision|SetCollisionEnabled (Variables|Default|GetLaunchableMesh) "NoCollision"))))

(fn PerformOverlapActorCheck
  (bind (_outhit _returnvalue) (Collision|BoxTraceForObjects (+ (Transformation|GetActorLocation) (Variables|Default|GetBoxTraceHeightOffset)) (+ (Transformation|GetActorLocation) (Variables|Default|GetBoxTraceHeightOffset)) (Math|Vector|MakeVector (ScalableFloat|GetValueatLevel (Variables|Hotfix|GetHF_RegrowthOverlapSizeX)) (ScalableFloat|GetValueatLevel (Variables|Hotfix|GetHF_RegrowthOverlapSizeY)) (ScalableFloat|GetValueatLevel (Variables|Hotfix|GetHF_RegrowthOverlapSizeZ))) "0, 0, 0" (Utilities|Array|MakeArray "ObjectTypeQuery3" "ObjectTypeQuery5" "ObjectTypeQuery1") false (Utilities|Array|MakeArray (Class|SimpleDestructiblePropExplosionProxy|GetSimpleDestructiblePropExplosionProxy)) "Persistent" true "(R=1.000000,G=0.000000,B=0.000000,A=1.000000)" "(R=0.000000,G=1.000000,B=0.000000,A=1.000000)" 0.0))
  (if _returnvalue
    ;[0.2144 0.6667 0.07986] We keep attempting to regrow each time this pass is false
    (Utilities|FlowControl|Sequence
      (:then_0
        (Utilities|Time|SetTimerbyEvent (EventDispatchers|CreateEvent :!event TryRegrowth) (ScalableFloat|GetValueatLevel (Variables|Hotfix|GetHF_CheckRegrowthInterval))))
      (:then_1
        (return false)))
    (else
      (return true))))

(fn ResetHealth
  (Health|SetHealth (Health|GetMaxHealth))
  (return))

(fn SetInstigatingActorWithCheck
  (params (Actor :type Actor%20Object%20Reference))
  (bind _asfort_player_state_athena (Utilities|Casting|CastToFortPlayerStateAthena Actor)
    (:then
      (Utilities|IsValid (PlayerState|GetCurrentPawn _asfort_player_state_athena)
        (:Is%20Valid
          (Variables|Default|SetInstigatingActor (PlayerState|GetCurrentPawn _asfort_player_state_athena)))
        (:Is%20Not%20Valid
          (Variables|Default|SetInstigatingActor Actor)))
      (return))
    (:CastFailed
      _)))

(fn ConstructionScript)
